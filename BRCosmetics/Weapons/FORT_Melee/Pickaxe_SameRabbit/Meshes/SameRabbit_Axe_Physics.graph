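# Physics asset for the SameRabbit two-handed pickaxe cosmetic, giving its hanging pendant physical bodies.
# A large capsule covers the handle, and four small capsules follow the dyn_pendant bone chain, with a wider one at the tip.
# Constraints link each pendant joint to the one above and the first to the handle, allowing limited swing and very little twist.
# The first joint has uneven swing limits, while the rest share equal ones.
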PhysicsAsset /BRCosmetics/Weapons/FORT_Melee/Pickaxe_SameRabbit/Meshes/SameRabbit_Axe_Physics
bodies {
 Melee_TwoHanded_Handle [{shapeType=Capsule, length=33.5426, radius=7.0386, center=<-0.4185 -3.2844 10.3244>, rotation=<rot 6.1016 -90 -0.0146>}]
 dyn_pendant_01_jnt [{shapeType=Capsule, length=0.8659, radius=0.505, shapeName=dyn_pendant_01_jnt_capsule, center=<0.899 0 0>, rotation=<rot 90 0 0>}]
 dyn_pendant_02_jnt [{shapeType=Capsule, length=0.3094, radius=0.5, shapeName=dyn_pendant_02_jnt_capsule, center=<0.4715 0 0>, rotation=<rot 90 0 0>}]
 dyn_pendant_03_jnt [{shapeType=Capsule, length=0.377, radius=0.5, shapeName=dyn_pendant_03_jnt_capsule, center=<0.4651 -0.0285 0.0277>, rotation=<rot 90 0 0>}]
 dyn_pendant_04_jnt [{
   shapeType=Capsule
   length=0.944
   radius=1.7101
   shapeName=dyn_pendant_04_jnt_capsule
   center=<2.1552 0 -0.0115>
   rotation=<rot 0 -90 -88.5316>
  }]
}
constraints [
 {bone1Name=dyn_pendant_01_jnt, bone2Name=Melee_TwoHanded_Handle, swing1LimitDegrees=10, swing2LimitDegrees=40, twistLimitDegrees=5}
 {bone1Name=dyn_pendant_02_jnt, bone2Name=dyn_pendant_01_jnt, swing1LimitDegrees=35, swing2LimitDegrees=35, twistLimitDegrees=5}
 {bone1Name=dyn_pendant_03_jnt, bone2Name=dyn_pendant_02_jnt, swing1LimitDegrees=35, swing2LimitDegrees=35, twistLimitDegrees=5}
 {bone1Name=dyn_pendant_04_jnt, bone2Name=dyn_pendant_03_jnt, swing1LimitDegrees=35, swing2LimitDegrees=35, twistLimitDegrees=5}
]
