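# Animation Blueprint for the SameRabbit pickaxe cosmetic, adding procedural secondary motion to its skeleton.
# A local-space ref pose passes through DefaultSlot into a RigidBody node using the axe physics asset, with overridden gravity and damped component motion.
# Three chained AnimDynamics nodes then layer low-alpha, angular-spring sway with overridden gravity onto pendant joints 02 to 04.
# The event graph holds no logic, and the Input Pose node is unconnected.
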AnimBlueprint /BRCosmetics/Weapons/FORT_Melee/Pickaxe_SameRabbit/Meshes/SameRabbit_Axe_AnimBP
target_skeleton=&/SameRabbit_Axe_Skeleton
graphs [
 AnimGraph {nodes [
   n1 = Local Space Ref Pose()
   n2 = RigidBody(ComponentPose=Local To Component(LocalPose=Slot 'DefaultSlot'(Source=@n1))) {
    bOverrideWorldGravity=true
    overridePhysicsAsset=&/SameRabbit_Axe_Physics
    alphaBoolBlend {blendOption=Linear}
    componentAppliedLinearAccClamp=<1000 1000 1000>
    componentLinearAccScale=<0.1 0.3 0.1>
    componentLinearVelScale=<0.18 0.33 0.18>
    overrideWorldGravity {z=-980}
    simSpaceSettings {maxAngularAcceleration=1100, maxAngularVelocity=270, maxLinearAcceleration=150, maxLinearVelocity=50, worldAlpha=1}
   }
   n3 = Anim Dynamics - Bone: dyn_pendant_02_jnt(ComponentPose=@n2, Alpha=0.3, ExternalForce="0,0,0") {
    angularDampingOverride=1
    angularSpringConstant=1000
    bAngularSpring=true
    bOverrideAngularDamping=true
    bUseGravityOverride=true
    gravityScale=0.35
    simulationSpace=RootRelative
    alphaBoolBlend {blendOption=Linear}
    boundBone {boneName=dyn_pendant_02_jnt}
    componentAppliedLinearAccClamp=<1000 1000 1000>
    componentLinearAccScale=<1 1 1>
    componentLinearVelScale=<1 1 1>
    gravityOverride {z=-650}
    physicsBodyDefinitions [{
      boundBone {boneName=dyn_pendant_02_jnt}
      boxExtents=<7 7 7>
      constraintSetup {angularLimitsMax=<30 30 30>, angularLimitsMin=<-30 -30 -30>, angularTarget=<0.5 0 0>}
      localJointOffset=<0.05 0 0>
     }]
   }
   n4 = Anim Dynamics - Bone: dyn_pendant_03_jnt(ComponentPose=@n3, Alpha=0.3, ExternalForce="0,0,0") {
    angularDampingOverride=1
    angularSpringConstant=1000
    bAngularSpring=true
    bOverrideAngularDamping=true
    bUseGravityOverride=true
    gravityScale=0.35
    simulationSpace=RootRelative
    alphaBoolBlend {blendOption=Linear}
    boundBone {boneName=dyn_pendant_03_jnt}
    componentAppliedLinearAccClamp=<1000 1000 1000>
    componentLinearAccScale=<1 1 1>
    componentLinearVelScale=<1 1 1>
    gravityOverride {z=-650}
    physicsBodyDefinitions [{
      boundBone {boneName=dyn_pendant_03_jnt}
      boxExtents=<7 7 7>
      constraintSetup {angularLimitsMax=<30 30 30>, angularLimitsMin=<-30 -30 -30>, angularTarget=<0.5 0 0>}
      localJointOffset=<0.05 0 0>
     }]
   }
   n5 = Anim Dynamics - Bone: dyn_pendant_04_jnt(ComponentPose=@n4, Alpha=0.3, ExternalForce="0,0,0") {
    angularDampingOverride=1
    angularSpringConstant=1000
    bAngularSpring=true
    bOverrideAngularDamping=true
    bUseGravityOverride=true
    gravityScale=0.35
    simulationSpace=RootRelative
    alphaBoolBlend {blendOption=Linear}
    boundBone {boneName=dyn_pendant_04_jnt}
    componentAppliedLinearAccClamp=<1000 1000 1000>
    componentLinearAccScale=<1 1 1>
    componentLinearVelScale=<1 1 1>
    gravityOverride {z=-650}
    physicsBodyDefinitions [{
      boundBone {boneName=dyn_pendant_04_jnt}
      boxExtents=<7 7 7>
      constraintSetup {angularLimitsMax=<30 30 30>, angularLimitsMin=<-30 -30 -30>, angularTarget=<0.5 0 0>}
      localJointOffset=<0.05 0 0>
     }]
   }
   Output Pose(Result=Component To Local(ComponentPose=@n5)) {layerGroup=None}
   AnimGraphNode_SubInput_0 = Input Pose()
  ]}
 EventGraph {nodes [
   TryGetPawnOwner()
   Event BlueprintUpdateAnimation()
  ]}
]
variables []
